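# Looping dark magic sound cue for the Season 33 Meet Ronin story quest, in the gadgets sound class with chest attenuation.
# A random node picks one of two looping waves, each behind a short random delay, then a modulator varies it before the mixer.
# The mixer's second, quieter input is a delay with no sound attached, so it adds nothing audible.
# Concurrency allows one instance per owner and blocks new ones.
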SoundCue /BattlePassS33_Quests/Audio/SFX/Quest_S33_MeetRonin/Quest_S33_Story_DarkMagic_Loop_Cue
root=Mixer_0
nodes [
 {id=Delay_0, children [null], properties {delayMax=0.2, delayMin=0.2}}
 {id=Delay_1, children [WavePlayer_0], properties {delayMax=0.3, delayMin=0.1}}
 {id=Delay_2, children [WavePlayer_1], properties {delayMax=0.3, delayMin=0.1}}
 {id=Mixer_0, children [Modulator_0,Delay_0], properties {inputVolume [1,0.4]}}
 {id=Modulator_0, children [Random_0], properties {pitchMax=1.02}}
 {id=Random_0, children [Delay_1,Delay_2], properties {weights [1,1]}}
 {
  id=WavePlayer_0
  properties {bLooping=true, soundWaveAssetPtr=&/Quest_S33_Story_DarkMagic_Loop_01}
 }
 {
  id=WavePlayer_1
  properties {bLooping=true, soundWaveAssetPtr=&/Quest_S33_Story_DarkMagic_Loop_02}
 }
]
properties {
 attenuationSettings=/Game/Sounds/Attenuation/Fort_Chests_Att
 bOverrideConcurrency=true
 maxDistance=1300
 soundClassObject=&Fort_Gadgets
 volumeMultiplier=1.5
 concurrencyOverrides {bLimitToOwner=true, maxCount=1, resolutionRule=PreventNew}
}
