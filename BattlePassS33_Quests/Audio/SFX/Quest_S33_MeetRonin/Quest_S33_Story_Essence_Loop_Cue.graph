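# Sound cue for the Season 33 Meet Ronin story quest, playing an essence loop with chest attenuation and the gadgets sound class.
# A Random node picks one of three branches, each starting after a short random delay: looping Loop_01, looping Loop_02, or a non-looping Loop_02 player.
# The pick passes through a Modulator for pitch variation and a single-input Mixer.
# A concurrency override allows one instance per owner and blocks new ones.
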SoundCue /BattlePassS33_Quests/Audio/SFX/Quest_S33_MeetRonin/Quest_S33_Story_Essence_Loop_Cue
root=Mixer_0
nodes [
 {id=Delay_1, children [WavePlayer_3], properties {delayMax=0.6, delayMin=0.2}}
 {id=Delay_2, children [WavePlayer_0], properties {delayMax=0.4, delayMin=0.1}}
 {id=Delay_3, children [WavePlayer_1], properties {delayMax=0.5, delayMin=0.3}}
 {id=Mixer_0, children [Modulator_0], properties {inputVolume [1]}}
 {id=Modulator_0, children [Random_0], properties {pitchMin=0.97}}
 {id=Random_0, children [Delay_1,Delay_2,Delay_3], properties {weights [1,1,1]}}
 {
  id=WavePlayer_0
  properties {bLooping=true, soundWaveAssetPtr=&/Quest_S33_Story_Essence_Loop_02}
 }
 {id=WavePlayer_1, properties {soundWaveAssetPtr=&Quest_S33_Story_Essence_Loop_02}}
 {
  id=WavePlayer_3
  properties {bLooping=true, soundWaveAssetPtr=&/Quest_S33_Story_Essence_Loop_01}
 }
]
properties {
 attenuationSettings=/Game/Sounds/Attenuation/Fort_Chests_Att
 bOverrideConcurrency=true
 maxDistance=1300
 soundClassObject=&Fort_Gadgets
 volumeMultiplier=1
 concurrencyOverrides {bLimitToOwner=true, maxCount=1, resolutionRule=PreventNew}
}
