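# Sound cue for the Season 33 Meet Ronin story quest's essence pickup SFX.
# A Random node picks one of three equally weighted pickup waves, feeding a default Modulator and a Mixer with an empty second input.
# It borrows the Hidden Star ambient loop attenuation and plays on the non-critical in-game UI sound class with boosted volume.
# Overridden concurrency allows one instance per owner and blocks new plays while one is active.
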SoundCue /BattlePassS33_Quests/Audio/SFX/Quest_S33_MeetRonin/Quest_S33_Story_Essence_Pickup_Cue
root=Mixer_0
nodes [
 {id=Mixer_0, children [Modulator_0,null], properties {inputVolume [1,1]}}
 {id=Modulator_0, children [Random_0]}
 {id=Random_0, children [WavePlayer_3,WavePlayer_0,WavePlayer_1], properties {weights [1,1,1]}}
 {id=WavePlayer_0, properties {soundWaveAssetPtr=&/Quest_S33_Story_Essence_Pickup_02}}
 {id=WavePlayer_1, properties {soundWaveAssetPtr=&/Quest_S33_Story_Essence_Pickup_03}}
 {id=WavePlayer_3, properties {soundWaveAssetPtr=&/Quest_S33_Story_Essence_Pickup_01}}
]
properties {
 attenuationSettings=/Game/Athena/Sounds/SupplyDrops/Fort_HiddenStar_Ambient_Loop_Att
 bOverrideConcurrency=true
 maxDistance=2500
 soundClassObject=&Fort_UI_InGame_NonCritical
 volumeMultiplier=3
 concurrencyOverrides {bLimitToOwner=true, maxCount=1, resolutionRule=PreventNew}
}
